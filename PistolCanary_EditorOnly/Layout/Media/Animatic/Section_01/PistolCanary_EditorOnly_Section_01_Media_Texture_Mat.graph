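# Masked, default-lit material for the Section 01 animatic media texture in the editor-only PistolCanary layout content.
# Its graph holds a single texture sample of the Section 01 media texture, with no output connections shown.
# Automatic setting of usage flags in the editor is turned off.
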Material /PistolCanary_EditorOnly/Layout/Media/Animatic/Section_01/PistolCanary_EditorOnly_Section_01_Media_Texture_Mat
graph {
 properties {bAutomaticallySetUsageInEditor=false, blendMode=BLEND_Masked, usedShadingModels=Default Lit}
 nodes [TextureSample() {texture=&/PistolCanary_EditorOnly_Section_01_Media_Texture}]
}
